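# Unlit, two-sided translucent flame material for Niagara sprite and mesh particles in the TwilightSpotShine backpack effects.
# Opacity multiplies panning smoke by a procedural, noise-split flame shape with edge and bottom fades, then depth fade and particle alpha.
# Emissive uses a panning mask over bounds-relative coordinates to look up a color gradient, tinted by particle color.
# World position offset adds normal-map and sine wobble to switchable spline thickening or bending, driven by Niagara dynamic parameters.
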Material /BRCosmetics/Accessories/FORT_Backpacks/Backpack_F_MED_TwilightSpotShine/FX/M_TwilightSpotShine_Flames
graph {
 properties {
  bUsedWithNiagaraMeshParticles=true
  bUsedWithNiagaraSprites=true
  blendMode=BLEND_Translucent
  pixelDepthOffsetMode=PDOM_Legacy
  shadingModel=MSM_Unlit
  twoSided=true
  usedShadingModels=Unlit
 }
 nodes [
  n1 = ScalarParameter() {parameterName=SplineThicken_WidthBase}
  n2 = VertexColor()
  n3 = MaterialFunctionCall(Float2=TextureCoordinate()) {materialFunction=&BreakOutFloat2Components}
  n4 = ScalarParameter() {defaultValue=3, parameterName=SplineThicken_WidthBaseFade_Power}
  n5 = MaterialFunctionCall(
   WidthBase=LinearInterpolate(A=Multiply(A=@n1, B=@n2.A), Alpha=Power(Base=@n3.G, Exp=@n4)) {constB=0}
   WidthTip=Multiply(A=ScalarParameter() {defaultValue=10, parameterName=SplineThicken_WidthTop}, B=@n2.A)
  ) {materialFunction=&SplineThicken}
  n6 = DynamicParameter() {
   desc=Parameters set by Niagara
   parameterIndex=1
   defaultValue=<rgba 1 1 50 1>
   paramNames [Smoke Pan Speed, WPO Speed, Sphere Mask Radius, WPO Intensity]
  }
  n7 = MaterialFunctionCall(X=@n6.Smoke Pan Speed, Y=@n6.Smoke Pan Speed, Z=@n6.Smoke Pan Speed) {materialFunction=&MakeFloat3}
  n8 = ParticlePositionWS()
  n9 = ScalarParameter() {defaultValue=10, group=WPO, parameterName=Interpolate Clamp}
  n10 = ComponentMask(None=OneMinus(None=TextureCoordinate())) {g=true}
  n11 = ScalarParameter() {defaultValue=3, group=WPO, parameterName=WPOBendPower}
  n12 = LinearInterpolate(
   B=Clamp(None=Subtract(A=@n7.Result, B=@n8), Min=Multiply(B=@n9) {constA=-1}, Max=@n9) {maxDefault=1000, minDefault=-1000}
   Alpha=Multiply(A=Power(Base=@n10, Exp=@n11), B=@n6.WPO Intensity)
  )
  n13 = StaticSwitchParameter(True=@n12, False=Constant()) {defaultValue=true, group=WPO, parameterName=Use_Interpolate WPO}
  n14 = Add(A=TextureCoordinate(), B=ParticleRandom())
  n15 = ComponentMask(None=VectorParameter() {group=WPO, parameterName=WPO_B_UVScale, defaultValue=<rgba 0 1 0 0>}.RGB) {g=true, r=true}
  n16 = AppendVector(A=Constant(), B=ScalarParameter() {defaultValue=0.5, group=WPO, parameterName=WPO_B_Speed})
  n17 = DynamicParameter() {defaultValue=<rgba 1 1 50 1>, paramNames [Smoke Pan Speed, WPO Speed, Sphere Mask Radius, WPO Intensity]}
  n18 = TextureSample(UVs=Panner(Coordinate=Multiply(A=@n14, B=@n15), Speed=Multiply(A=@n16, B=@n17.WPO Speed)) {speedY=0.5}) {samplerType=SAMPLERTYPE_Normal, texture=&T_Weap_Distort_N_02}
  n19 = Add(A=TextureCoordinate(), B=ParticleRandom())
  n20 = ComponentMask(None=VectorParameter() {group=WPO, parameterName=WPO_A_UVScale, defaultValue=<rgba 0 1.5 0 0>}.RGB) {g=true, r=true}
  n21 = AppendVector(A=Constant(), B=ScalarParameter() {defaultValue=0.5, group=WPO, parameterName=WPO_A_Speed})
  n22 = TextureSample(UVs=Panner(Coordinate=Multiply(A=@n19, B=@n20), Speed=Multiply(A=@n21, B=@n17.WPO Speed)) {speedY=0.5}) {samplerType=SAMPLERTYPE_Normal, texture=&T_Weap_Distort_N_02}
  n23 = Add(
   A=Multiply(A=ScalarParameter() {defaultValue=1, group=WPO, parameterName=WPO_B_Intensity}, B=LinearInterpolate(Alpha=@n18.R) {constB=1.5})
   B=Multiply(A=ScalarParameter() {defaultValue=1, group=WPO, parameterName=WPO_A_Intensity}, B=LinearInterpolate(Alpha=@n22.R) {constB=1.5})
  )
  n24 = OneMinus(None=ComponentMask(None=TextureCoordinate()) {g=true})
  n25 = Saturate(None=Multiply(A=Power(Base=@n24) {constExponent=1}))
  n26 = OneMinus(None=ComponentMask(None=TextureCoordinate()) {g=true})
  n27 = Clamp(None=Multiply(A=Power(Base=@n26) {constExponent=1})) {maxDefault=0.5}
  n28 = Time()
  n29 = Add(A=ComponentMask(None=TextureCoordinate()) {g=true}, B=Add(A=Multiply(A=@n17.WPO Speed, B=@n28), B=ParticleRandom()))
  n30 = Multiply(A=Saturate(None=Sine(None=@n29))) {constB=0.3}
  n31 = Add(
   A=Multiply(A=Multiply(A=@n23, B=@n25), B=Constant3Vector() {constant=<rgba 1.5 0 0 1>})
   B=Multiply(A=Multiply(A=@n27, B=@n30), B=Constant3Vector() {constant=<rgba 0 0 1 1>})
  ) {constB=0}
  n32 = ScalarParameter() {defaultValue=1, group=WPO, parameterName=WPO Intensity}
  n33 = DynamicParameter() {defaultValue=<rgba 1 1 50 1>, paramNames [Smoke Pan Speed, WPO Speed, Sphere Mask Radius, WPO Intensity]}
  Add(
   A=StaticSwitchParameter(
    True=Add(A=@n5.WorldPositionOffset, B=@n12)
    False=StaticSwitchParameter(True=@n5.WorldPositionOffset, False=@n13) {parameterName=Use_SplineThicken}
   ) {parameterName=Use_SplineThickenWithInterpolateWPO}
   B=Multiply(A=Transform(None=@n31), B=Multiply(A=@n32, B=@n33.WPO Intensity)) {constB=12}
  )
  n34 = Constant()
  n35 = MaterialFunctionCall() {materialFunction=&ObjectLocalBounds}
  n36 = Subtract(
   A=TransformPosition(None=WorldPosition() {worldPositionShaderOffset=WPT_ExcludeAllShaderOffsets}.XYZ) {transformSourceType=TRANSFORMPOSSOURCE_World}
   B=@n35.Local Bounds Minimum
  )
  n37 = ComponentMask(None=ConstantBiasScale(None=Divide(A=@n36, B=@n35.Local Bounds Size))) {b=true}
  n38 = TextureSampleParameter2D() {
   group=06_Advanced Emissive
   parameterName=UVScaleMap
   samplerSource=SSM_Wrap_WorldGroupSettings
   samplerType=SAMPLERTYPE_Masks
   texture=&T_WhiteMask
  }
  n39 = Add(A=@n38.RGB, B=OneMinus(None=ScalarParameter() {defaultValue=1, group=06_Advanced Emissive, parameterName=UVScaleMapIntensity}))
  n40 = ScalarParameter() {defaultValue=1, group=06_Advanced Emissive, parameterName=EmissivePannerScale}
  n41 = StaticSwitchParameter(True=Multiply(A=Saturate(None=@n39), B=@n40), False=@n40) {
   defaultValue=true
   desc="This allows modification of the panning EmissiveGradient texture direction using a RG \"UV\" texture.\r\nIf \"WorldAlignedGradient\" is true, it uses RGB."
   group=06_Advanced Emissive
   parameterName=UseUVScaleMap
  }
  n42 = ComponentMask(None=@n41) {
   desc="Kendall added this node on 5/27/2020.\r\nPreviously, this just piped the 3-vector value from\r\nthe texture sample straight into the UV's, which\r\ncaused the material to fail compile"
   g=true
   r=true
  }
  n43 = Multiply(A=Saturate(None=OneMinus(None=@n37)), B=@n42)
  n44 = ScalarParameter() {group=06_Advanced Emissive, parameterName=GradientSpeed_X, sortPriority=2}
  n45 = ScalarParameter() {defaultValue=0.08, group=06_Advanced Emissive, parameterName=GradientSpeed_Y, sortPriority=2}
  n46 = TextureObjectParameter() {
   group=06_Advanced Emissive
   parameterName=EmissiveGradient
   samplerType=SAMPLERTYPE_Masks
   sortPriority=1
   texture=/Game/Global/Textures/Default/Masks_and_Gradients/T_VertGradient_MASK
  }
  n47 = TextureSample(UVs=Panner(Coordinate=Add(A=@n34, B=@n43), Speed=AppendVector(A=@n44, B=@n45)), Tex=@n46) {samplerType=SAMPLERTYPE_Masks, texture=/Game/Weapons/WeaponSkins/Textures/T_Swirly_Noise_highContrast}
  n48 = TextureSampleParameter2D(UVs=ComponentMask(None=@n47.RGB) {g=true, r=true}) {
   group=06_Advanced Emissive
   parameterName=PanningEmissiveColorGradient
   sortPriority=6
   texture=/BRCosmetics/Accessories/FORT_Backpacks/Backpack_F_MED_TwilightSpotShine/Textures/T_TwilightSpotShine_Backpack_Gradient
  }
  n49 = VectorParameter() {group=06_Advanced Emissive, parameterName=AdvancedEmissiveColor, sortPriority=15, defaultValue=<rgba 1 1 1 1>}
  n50 = Multiply(
   A=@n49.RGB
   B=LinearInterpolate(
    A=ScalarParameter() {group=06_Advanced Emissive, parameterName=AdvancedEmissiveBrightnessMin, sortPriority=16}
    B=ScalarParameter() {defaultValue=1, group=06_Advanced Emissive, parameterName=AdvancedEmissiveBrightnessMax, sortPriority=16}
    Alpha=@n47.RGB
   )
  )
  MaterialFunctionCall(Emissive=Multiply(
    A=StaticSwitchParameter(True=@n48.RGB, False=@n50) {defaultValue=true, group=06_Advanced Emissive, parameterName=UseGradientForColor, sortPriority=0}
    B=ParticleColor().RGB
   )) {materialFunction=&MF_FX_EmissiveAdjustments}
  n51 = Constant()
  n52 = ParticleRandom()
  n53 = MaterialFunctionCall(UVs=Add(A=TextureCoordinate(), B=AppendVector(A=@n51, B=@n52)), "Rotation Angle (0-1)"=Constant() {r=-0.25}) {materialFunction=&CustomRotator}
  n54 = ScalarParameter() {defaultValue=1, group=Smoke, parameterName=Smoke Tex B_ScaleU}
  n55 = ScalarParameter() {defaultValue=2, parameterName=Smoke Tex B_Scale_U_Mult}
  n56 = AppendVector(
   A=LinearInterpolate(A=@n54, B=Multiply(A=@n54, B=@n55), Alpha=@n52) {constAlpha=1}
   B=ScalarParameter() {defaultValue=1, group=Smoke, parameterName=Smoke Tex B_ScaleV}
  )
  n57 = ScalarParameter() {defaultValue=0.5, group=Smoke, parameterName=Smoke Tex B Pan}
  n58 = ScalarParameter() {defaultValue=2, parameterName=Smoke Tex B Pan Mult}
  n59 = DynamicParameter() {defaultValue=<rgba 1 1 50 1>, paramNames [Smoke Pan Speed, WPO Speed, Sphere Mask Radius, WPO Intensity]}
  n60 = Multiply(A=LinearInterpolate(A=@n57, B=Multiply(A=@n57, B=@n58) {constB=2}, Alpha=@n52), B=@n59.Smoke Pan Speed)
  n61 = Constant()
  n62 = TextureSampleParameter2D(UVs=Panner(
    Coordinate=MaterialFunctionCall(UVs=@n53.Rotated Values, Texture Scale=@n56) {materialFunction=&ScaleUVsByCenter}.UVs
    Speed=AppendVector(A=@n60, B=@n61)
   )) {group=Smoke, parameterName=Smoke A Tex, texture=/Game/Effects/Fort_Effects/Textures/Smoke/T_Smoke_Trail_Soft_Chromatic}
  n63 = VectorParameter() {group=Fire Shape, parameterName=Fire_MacroShape_UVScale, sortPriority=5, defaultValue=<rgba 1 1 0 0>}
  n64 = MaterialFunctionCall(Float2=Multiply(A=TextureCoordinate(), B=ComponentMask(None=@n63.RGB) {g=true, r=true})) {materialFunction=&BreakOutFloat2Components}
  n65 = Power(
   Base=Sine(None=Multiply(A=@n64.R) {constB=0.5})
   Exp=LinearInterpolate(
    A=ScalarParameter() {defaultValue=3, group=Fire Shape, parameterName=Edge Fade Value Top, sortPriority=3}
    B=ScalarParameter() {defaultValue=2, group=Fire Shape, parameterName=Edge Fade Value Bottom, sortPriority=4}
    Alpha=@n64.G
   ) {constA=6, constB=2}
  ) {constExponent=4}
  n66 = VectorParameter() {group=BottomFade, parameterName=Bottom Fade_SphereMask_A_Scale, sortPriority=2, defaultValue=<rgba 1.2 1 0 0>}
  n67 = Multiply(A=TextureCoordinate(), B=ComponentMask(None=@n66.RGB) {g=true, r=true})
  n68 = VectorParameter() {group=BottomFade, parameterName=Bottom Fade_SphereMask_B, sortPriority=3, defaultValue=<rgba 0.6 0.05 0 0>}
  n69 = ComponentMask(None=@n68.RGB) {g=true, r=true}
  n70 = ScalarParameter() {defaultValue=0.95, group=BottomFade, parameterName=BottomFade_Radius, sortPriority=1}
  n71 = Saturate(None=Multiply(A=SphereMask(A=@n67, B=@n69, Radius=@n70) {attenuationRadius=0.95, hardnessPercent=0}) {constB=5})
  n72 = ScalarParameter() {defaultValue=1.5, group=BottomFade, parameterName=Bottom Fade}
  n73 = Power(
   Base=Multiply(A=Multiply(A=@n65, B=@n64.G), B=LinearInterpolate(B=@n71, Alpha=@n72) {constA=1})
   Exp=LinearInterpolate(Alpha=@n64.G) {constA=2}
  )
  n74 = ScalarParameter() {defaultValue=1, group=Fire Shape, parameterName=Fire Split Contrast, sortPriority=1}
  n75 = OneMinus(None=Sine(None=Multiply(A=@n64.R) {constB=0.45}))
  n76 = Add(A=@n52, B=TextureCoordinate() {vTiling=1.5})
  n77 = ComponentMask(None=VectorParameter() {group=Distortion, parameterName=Distortion UV Scale, sortPriority=3, defaultValue=<rgba 1 1 0 0>}.RGB) {g=true, r=true}
  n78 = TextureSampleParameter2D(UVs=Panner(Coordinate=Multiply(A=@n76, B=@n77)) {speedY=1}) {
   group=Distortion
   parameterName=Distortion Normal
   samplerType=SAMPLERTYPE_Normal
   sortPriority=2
   texture=&T_Weap_Distort_N_02
  }
  n79 = LinearInterpolate(Alpha=@n78.R) {constA=0.5, constB=1.5}
  n80 = Round(None=Add(A=Power(Base=@n64.G, Exp=@n74), B=Power(Base=@n75, Exp=@n79) {constExponent=0.5}))
  n81 = ScalarParameter() {defaultValue=0.33, group=Fire Shape, parameterName=Fire Split}
  n82 = DynamicParameter() {parameterIndex=2, defaultValue=<rgba 1 1 50 1>, paramNames [Smoke Pan Speed, WPO Speed, Sphere Mask Radius, WPO Intensity]}
  n83 = Multiply(A=Multiply(A=Multiply(A=@n62.R, B=@n73), B=LinearInterpolate(B=@n80, Alpha=@n81) {constA=1, constB=0}), B=@n82.Smoke Pan Speed) {constB=3}
  n84 = DepthFade(Opacity=@n83, FadeDistance=ScalarParameter() {defaultValue=5, group=Opacity, parameterName=DepthFadeDistance})
  Multiply(A=ShadingPathSwitch(Default=Multiply(A=@n84), Mobile=@n83) {desc=Migrated from FeatureLevelSwitch}, B=ParticleColor().A)
  n85 = DynamicParameter() {parameterIndex=1, defaultValue=<rgba 1 1 50 1>, paramNames [Smoke Pan Speed, WPO Speed, Sphere Mask Radius, WPO Intensity]}
  n86 = LinearInterpolate(
   A=ScalarParameter() {defaultValue=1, group=WPO, parameterName=WPO UV Scale Mult}
   B=@n85.Smoke Pan Speed
   Alpha=ScalarParameter() {group=WPO, parameterName=Use Dynamic WPO UV Scale}
  )
  Multiply(A=@n86, B=@n20)
  Multiply(A=@n15, B=@n86)
  PerInstanceRandom()
 ]
}
