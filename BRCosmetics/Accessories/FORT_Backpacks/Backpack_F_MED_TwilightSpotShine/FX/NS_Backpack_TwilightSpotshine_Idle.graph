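# Niagara system for the idle flame effect on the TwilightSpotshine backpack cosmetic.
# One emitter bursts ten randomly scaled, oriented mesh flames that stay pinned to the skeletal mesh at offsets from a hand-placed array.
# A time-interpolated, lagging copy of each particle position feeds dynamic material parameters, alongside randomized values.
# Camera offset varies with the dot product of view vector and system X axis.
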NiagaraSystem /BRCosmetics/Accessories/FORT_Backpacks/Backpack_F_MED_TwilightSpotShine/FX/NS_Backpack_TwilightSpotshine_Idle
dependencies {
 usedDataInterfaces [
  &NiagaraDataInterfaceCamera
  &NiagaraDataInterfaceArrayFloat3
  &NiagaraDataInterfaceCurve
  &NiagaraDataInterfaceSkeletalMesh
 ]
 usedDynamicInputs []
 usedModules [
  &self:SystemSpawnScript.SystemScriptSource.SystemScriptGraph.NiagaraNodeAssignment_0.SetVariables_#1
  &SystemState
 ]
}
emitters {Flames {
  emitterUpdateScript [
   &EmitterState(
    Inactive Response=::NewEnumerator1
    Life Cycle Mode=::NewEnumerator1
    Loop Behavior=::NewEnumerator1
    Loop Duration Mode=::NewEnumerator1
    UseLoopDelay=false
   )
   &SpawnBurst_Instantaneous(Spawn Count=10, Use Spawn Probability=false, UseLoopCountLimit=false)
  ]
  particleSpawnScript [
   &InitializeParticle(
    Color=<rgba 1 1 1 0.075>
    Mesh Scale=&RandomRangeVector(Minimum=<0.7 0.7 0.35>, Maximum=<0.5 0.5 0.3>)
    Mesh Scale Mode=::NewEnumerator4
    Sprite Size Mode=::NewEnumerator3
    UsePositionOffset=false
    Write Material Random=true
   )
   &InitialMeshOrientation(
    Mesh Orientation Mode=::NewEnumerator2
    Rotation=&RandomRangeVector(Minimum=<0.03 -0.05 0>, Maximum=<0.05 0.05 0>)
    Rotation Coordinate Space=::NewEnumerator1
    Use Rotation Vector=true
   )
  ]
  particleUpdateScript [
   &ParticleState(Kill Particles When Lifetime Has Elapsed=false, Let Infinitely Lived Particles Die When Emitter Deactivates=true)
   &CameraOffset(Camera Offset Amount=&Lerp_Float(
     A=0
     B=-2
     Alpha=&NormalizeFloat(
      Value=&DotProduct(
       Vector A=&ReturnCameraVector(Camera Vectors=::NewEnumerator4, Coordinate Space=::Simulation, Origin Position=$Particles.Position)
       Vector B=$Engine.Owner.SystemXAxis
      )
      Min=0
      Max=1
     )
    ))
   &SkeletalMeshLocation(
    MaskSpawnGroup=false
    Mesh Sampling Type=::NewEnumerator5
    Sampled Position Offset=&SelectVectorFromArray(
     Array Sampling Mode=::NewEnumerator0
     Direct Array Index=?
     Vector Selection Array=DataInterface() {propertyValues {FloatData [
        <3.25 1 12>
        <-3.5 1 16>
        <3 1 -18>
        <-3.1 1 -10>
        <1.5 1 -5>
        <-3.5 1 1>
        <-1.8133 1 7.1206>
        <3.5 1 -5.9927>
        <-0.6484 1 8.7321>
        <-3.5 1 -2.0915>
        <1.5794 1 -0.3744>
       ]}}
     Random Seed=0
    )
    Skeletal Mesh=$System.Skeletal Mesh
    Socket Sampling Mode=::NewEnumerator5
   )
   &InterpolateOverTime(
    Rate Of Change:NiagaraFloat=&RandomRangeFloat(Minimum=7, Maximum=15, Evaluation Type=::NewEnumerator1)
    Target Vector Value:Vector3f=&ConvertPositionToVector(Input Position=$Particles.Position, Convert To Absolute Large World Vector=true)
    Variable Type=::NewEnumerator2
   )
   /SetVariables(Particles.SpriteDynParams:Vector4f=&MakeVector4(
     X=0
     Y=&FloatFromCurve(
      FloatCurve=DataInterface() {propertyValues {Curve {keys [{tangentMode=RCTM_User, time=0, value=0}, {tangentMode=RCTM_User, time=1, value=0.5}]}}}
      Scale Curve=&FloatFromCurve(FloatCurve=DataInterface() {propertyValues {Curve {keys [<at 0 1.7 RCIM_Cubic>,<at 1 1.7 RCIM_Cubic>]}}})
     )
     Z=0
     W=0
    )) {isSetParameters=true}
   /SetVariables(Particles.Interpolate Dynamic=&MakeFloat4FromVectorAndFloat("Vector (RGB)"=$StackContext.InterpolateOverTime.InterpolateOverTime_V3_.Moving Average, "Float (Alpha)"=1)) {isSetParameters=true}
   /SetVariables(Particles.InterpolatedPosition:NiagaraPosition=$Particles.InterpolateOverTime.InterpolateOverTime_V3_.Moving Average) {isSetParameters=true}
   &DynamicMaterialParameters(
    Index 0 Param 1=1
    Index 0 Param 2=&RandomRangeFloat(Minimum=2, Maximum=3)
    Index 0 Param 3=5
    Index 0 Param 4=4
    Index 1 Param 1=&MakeFloatFromVector4(Vector 4=$Particles.Interpolate Dynamic, Channel=::NewEnumerator0)
    Index 1 Param 2=&MakeFloatFromVector4(Vector 4=$Particles.Interpolate Dynamic, Channel=::NewEnumerator1)
    Index 1 Param 3=&MakeFloatFromVector4(Vector 4=$Particles.Interpolate Dynamic, Channel=::NewEnumerator2)
    Index 1 Param 4=1
    Index 2 Param 1=&RandomRangeFloat(Minimum=2, Maximum=4)
    Index 2 Param 2=&RandomRangeFloat(Minimum=0.5, Maximum=1)
    Module.Index0Float0Write=true
    Module.Index0Float1Write=true
    Module.Index0Float2Write=true
    Module.Index0Float3Write=true
    Module.Index1Float0Write=true
    Module.Index1Float1Write=true
    Module.Index1Float2Write=true
    Module.Index1Float3Write=true
    Module.Index2Float0Write=true
    Module.Index2Float1Write=true
    Module.Index2Float2Write=true
    Module.Index2Float3Write=true
    Module.Index3Float0Write=true
    Module.Index3Float1Write=true
    Module.Index3Float2Write=true
    Module.Index3Float3Write=true
    Write Parameter Index 1=::NewEnumerator1
    Write Parameter Index 2=::NewEnumerator1
    Write Parameter Index 3=::NewEnumerator1
   )
  ]
  rendererClasses [&NiagaraMeshRendererProperties]
  renderers [{rendererClass=&NiagaraMeshRendererProperties, rendererIndex=0}]
 }}
summary {
 systemName=NS_Backpack_TwilightSpotshine_Idle
 emitters [{emitterName=Flames, rendererClasses [&NiagaraMeshRendererProperties]}]
 userVariables []
}
